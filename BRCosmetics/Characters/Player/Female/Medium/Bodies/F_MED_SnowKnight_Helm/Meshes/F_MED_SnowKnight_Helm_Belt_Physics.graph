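# Physics asset for the belt of the female medium SnowKnight Helm body cosmetic, with eight hanging flaps at front and back, left and right.
# Small base capsules are locked to the pelvis, and each anchors a dyn flap capsule through narrow swing and twist limits.
# Larger pelvis, glute and thigh capsules approximate the body.
# Glutes and right thigh are locked to the pelvis, while the left thigh has limited rotation.
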PhysicsAsset /BRCosmetics/Characters/Player/Female/Medium/Bodies/F_MED_SnowKnight_Helm/Meshes/F_MED_SnowKnight_Helm_Belt_Physics
bodies {
 base_flap_01_back_L [{shapeType=Capsule, length=0.505, radius=0.505, shapeName=base_flap_01_back_L_capsule}]
 base_flap_01_back_R [{shapeType=Capsule, length=0.505, radius=0.505, shapeName=base_flap_01_back_R_capsule}]
 base_flap_01_front_L [{shapeType=Capsule, length=0.505, radius=0.505, shapeName=base_flap_01_front_L_capsule}]
 base_flap_01_front_R [{shapeType=Capsule, length=0.505, radius=0.505, shapeName=base_flap_01_front_R_capsule}]
 base_flap_02_back_L [{shapeType=Capsule, length=0.505, radius=0.505, shapeName=base_flap_02_back_L_capsule}]
 base_flap_02_back_R [{shapeType=Capsule, length=0.505, radius=0.505, shapeName=base_flap_02_back_R_capsule}]
 base_flap_02_front_L [{shapeType=Capsule, length=0.505, radius=0.505, shapeName=base_flap_02_front_L_capsule}]
 base_flap_02_front_R [{shapeType=Capsule, length=0.505, radius=0.505, shapeName=base_flap_02_front_R_capsule}]
 deform_glute_l [{
   shapeType=Capsule
   length=4.3
   radius=5.7
   shapeName=deform_glute_l_capsule
   center=<5.152 -1.8714 2.4199>
   rotation=<rot 11.8309 3.6353 -20.6798>
  }]
 deform_glute_r [{
   shapeType=Capsule
   length=4.3
   radius=5.7
   shapeName=deform_glute_l_capsule
   center=<-5.152 1.8714 -2.4199>
   rotation=<rot 11.8309 3.6353 159.3202>
  }]
 dyn_flap_01_back_L [{shapeType=Capsule, length=3.7, radius=1, shapeName=dyn_flap_01_back_R_capsule, center=<3.9629 0 0>}]
 dyn_flap_01_back_R [{shapeType=Capsule, length=3.7, radius=1, shapeName=dyn_flap_01_back_R_capsule, center=<3.9629 0 0>, rotation=<rot 90 0 0>}]
 dyn_flap_01_front_L [{shapeType=Capsule, length=1.6, radius=1, shapeName=dyn_flap_01_front_L_capsule, center=<5.9985 0 0>}]
 dyn_flap_01_front_R [{shapeType=Capsule, length=1.6, radius=1, shapeName=dyn_flap_01_front_L_capsule, center=<5.9985 0 0>}]
 dyn_flap_02_back_L [{shapeType=Capsule, length=3.7, radius=1, shapeName=dyn_flap_02_back_R_capsule, center=<3.9629 0 0>}]
 dyn_flap_02_back_R [{shapeType=Capsule, length=3.7, radius=1, shapeName=dyn_flap_02_back_R_capsule, center=<3.9629 0 0>, rotation=<rot 90 0 0>}]
 dyn_flap_02_front_L [{shapeType=Capsule, length=1.6, radius=1, shapeName=dyn_flap_02_front_L_capsule, center=<6.8883 0 0>}]
 dyn_flap_02_front_R [{shapeType=Capsule, length=1.6, radius=1, shapeName=dyn_flap_02_front_L_capsule, center=<5.6002 0 0>}]
 pelvis [{shapeType=Capsule, length=12.7, radius=11.3, shapeName=pelvis_capsule, center=<-6.1447 -2.2685 0>}]
 thigh_l [{shapeType=Capsule, length=12.4, radius=8.8, shapeName=thigh_l_capsule, center=<-10.5919 -1.0599 0.524>, rotation=<rot 90 178.8065 172.5744>}]
 thigh_r [{shapeType=Capsule, length=11.8, radius=9.4, shapeName=thigh_l_capsule, center=<8.8935 1.2625 -0.524>, rotation=<rot 90 180 -3.4274>}]
}
constraints [
 {bone1Name=dyn_flap_01_back_R, bone2Name=base_flap_01_back_R, swing1LimitDegrees=20, swing2LimitDegrees=15, twistLimitDegrees=1}
 {bone1Name=dyn_flap_01_front_L, bone2Name=base_flap_01_front_L, swing1LimitDegrees=20, swing2LimitDegrees=10, twistLimitDegrees=10}
 {bone1Name=dyn_flap_02_back_R, bone2Name=base_flap_02_back_R, swing1LimitDegrees=20, swing2LimitDegrees=5, twistLimitDegrees=1}
 {bone1Name=dyn_flap_02_front_L, bone2Name=base_flap_02_front_L, swing1LimitDegrees=25, swing2LimitDegrees=10, twistLimitDegrees=5}
 {
  bone1Name=base_flap_02_back_R
  bone2Name=pelvis
  swing1LimitDegrees=45
  swing1Motion=Locked
  swing2LimitDegrees=45
  swing2Motion=Locked
  twistLimitDegrees=45
  twistMotion=Locked
 }
 {
  bone1Name=base_flap_01_back_R
  bone2Name=pelvis
  swing1LimitDegrees=45
  swing1Motion=Locked
  swing2LimitDegrees=45
  swing2Motion=Locked
  twistLimitDegrees=45
  twistMotion=Locked
 }
 {
  bone1Name=base_flap_02_back_L
  bone2Name=pelvis
  swing1LimitDegrees=45
  swing1Motion=Locked
  swing2LimitDegrees=45
  swing2Motion=Locked
  twistLimitDegrees=45
  twistMotion=Locked
 }
 {
  bone1Name=base_flap_01_back_L
  bone2Name=pelvis
  swing1LimitDegrees=45
  swing1Motion=Locked
  swing2LimitDegrees=45
  swing2Motion=Locked
  twistLimitDegrees=45
  twistMotion=Locked
 }
 {
  bone1Name=base_flap_02_front_R
  bone2Name=pelvis
  swing1LimitDegrees=45
  swing1Motion=Locked
  swing2LimitDegrees=45
  swing2Motion=Locked
  twistLimitDegrees=45
  twistMotion=Locked
 }
 {
  bone1Name=base_flap_01_front_R
  bone2Name=pelvis
  swing1LimitDegrees=45
  swing1Motion=Locked
  swing2LimitDegrees=45
  swing2Motion=Locked
  twistLimitDegrees=45
  twistMotion=Locked
 }
 {
  bone1Name=base_flap_01_front_L
  bone2Name=pelvis
  swing1LimitDegrees=45
  swing1Motion=Locked
  swing2LimitDegrees=45
  swing2Motion=Locked
  twistLimitDegrees=45
  twistMotion=Locked
 }
 {
  bone1Name=base_flap_02_front_L
  bone2Name=pelvis
  swing1LimitDegrees=45
  swing1Motion=Locked
  swing2LimitDegrees=45
  swing2Motion=Locked
  twistLimitDegrees=45
  twistMotion=Locked
 }
 {
  bone1Name=deform_glute_l
  bone2Name=pelvis
  swing1LimitDegrees=45
  swing1Motion=Locked
  swing2LimitDegrees=45
  swing2Motion=Locked
  twistLimitDegrees=45
  twistMotion=Locked
 }
 {
  bone1Name=deform_glute_r
  bone2Name=pelvis
  swing1LimitDegrees=45
  swing1Motion=Locked
  swing2LimitDegrees=45
  swing2Motion=Locked
  twistLimitDegrees=45
  twistMotion=Locked
 }
 {bone1Name=dyn_flap_02_back_L, bone2Name=base_flap_02_back_L, swing1LimitDegrees=35, swing2LimitDegrees=6, twistLimitDegrees=1}
 {bone1Name=dyn_flap_01_back_L, bone2Name=base_flap_01_back_L, swing1LimitDegrees=20, swing2LimitDegrees=10, twistLimitDegrees=1}
 {bone1Name=thigh_l, bone2Name=pelvis, swing1LimitDegrees=10, swing2LimitDegrees=10, twistLimitDegrees=5}
 {
  bone1Name=thigh_r
  bone2Name=pelvis
  swing1LimitDegrees=45
  swing1Motion=Locked
  swing2LimitDegrees=45
  swing2Motion=Locked
  twistLimitDegrees=45
  twistMotion=Locked
 }
 {bone1Name=dyn_flap_02_front_R, bone2Name=base_flap_02_front_R, swing1LimitDegrees=20, swing2LimitDegrees=10, twistLimitDegrees=5}
 {bone1Name=dyn_flap_01_front_R, bone2Name=base_flap_01_front_R, swing1LimitDegrees=20, swing2LimitDegrees=10, twistLimitDegrees=10}
]
