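# Physics asset for the SnowKnight Helm female medium body cosmetic, giving secondary motion to its jacket, collar and left-arm pauldron.
# Capsules on head, spine and upper arms are locked together as an anchor and collision frame.
# Sleeve, side, collar and some upper-jacket dynamic bones hang from locked base bones with tight swing/twist limits; the pauldron joint is fully free.
# Right-side sleeve bases and twist bone attach to spine_05, unlike the left's arm chain.
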PhysicsAsset /BRCosmetics/Characters/Player/Female/Medium/Bodies/F_MED_SnowKnight_Helm/Meshes/F_MED_SnowKnight_Helm_Physics
bodies {
 Collar_back_jacket [{shapeType=Capsule, length=3.7, radius=0.7, shapeName=Collar_back_jacket_capsule, center=<2.6189 0 0>}]
 base_Collar_back_jacket [{shapeType=Capsule, length=1, radius=1, shapeName=base_Collar_back_jacket_capsule}]
 base_Sleeve_jacket_01_l [{shapeType=Capsule, length=1, radius=1, shapeName=base_Sleeve_jacket_01_l_capsule}]
 base_Sleeve_jacket_01_r [{shapeType=Capsule, length=0.505, radius=0.505, shapeName=base_Sleeve_jacket_01_r_capsule, rotation=<rot 0 -90 0>}]
 base_Sleeve_jacket_02_l [{shapeType=Capsule, length=1, radius=1, shapeName=base_Sleeve_jacket_02_l_capsule}]
 base_Sleeve_jacket_02_r [{shapeType=Capsule, length=0.505, radius=0.505, shapeName=base_Sleeve_jacket_02_r_capsule, rotation=<rot 0 -90 0>}]
 base_Sleeve_jacket_03_l [{shapeType=Capsule, length=1, radius=1, shapeName=base_Sleeve_jacket_03_l_capsule}]
 base_Sleeve_jacket_03_r [{shapeType=Capsule, length=0.505, radius=0.505, shapeName=base_Sleeve_jacket_03_r_capsule, rotation=<rot 0 -90 0>}]
 base_Sleeve_jacket_04_l [{shapeType=Capsule, length=1, radius=1, shapeName=base_Sleeve_jacket_04_l_capsule}]
 base_Sleeve_jacket_04_r [{shapeType=Capsule, length=0.505, radius=0.505, shapeName=base_Sleeve_jacket_04_r_capsule, rotation=<rot 0 -90 0>}]
 base_Sleeve_jacket_05_l [{shapeType=Capsule, length=1, radius=1, shapeName=base_Sleeve_jacket_05_l_capsule}]
 base_Sleeve_jacket_05_r [{shapeType=Capsule, length=0.505, radius=0.505, shapeName=base_Sleeve_jacket_05_r_capsule, rotation=<rot 0 -90 0>}]
 base_Sleeve_jacket_06_l [{shapeType=Capsule, length=1, radius=1, shapeName=base_Sleeve_jacket_06_l_capsule}]
 base_Sleeve_jacket_06_r [{shapeType=Capsule, length=0.505, radius=0.505, shapeName=base_Sleeve_jacket_06_r_capsule, rotation=<rot 0 -90 0>}]
 base_Sleeve_jacket_07_l [{shapeType=Capsule, length=1, radius=1, shapeName=base_Sleeve_jacket_07_l_capsule}]
 base_Sleeve_jacket_07_r [{shapeType=Capsule, length=0.505, radius=0.505, shapeName=base_Sleeve_jacket_07_r_capsule, rotation=<rot 0 -90 0>}]
 base_Sleeve_jacket_08_l [{shapeType=Capsule, length=1, radius=1, shapeName=base_Sleeve_jacket_08_l_capsule}]
 base_Sleeve_jacket_08_r [{shapeType=Capsule, length=0.505, radius=0.505, shapeName=base_Sleeve_jacket_08_r_capsule, rotation=<rot 0 -90 0>}]
 base_side_jacket_l [{shapeType=Capsule, length=0.505, radius=0.505, shapeName=base_side_jacket_l_capsule}]
 base_side_jacket_r [{shapeType=Capsule, length=0.505, radius=0.505, shapeName=base_side_jacket_r_capsule}]
 base_upper_jacket_03_r [{shapeType=Capsule, length=0.505, radius=0.505, shapeName=base_upper_jacket_03_r_capsule, rotation=<rot 0 -90 0>}]
 dyn_Pauldron [{shapeType=Capsule, length=3.4116, radius=1.3774, shapeName=dyn_Pauldron_capsule, center=<12.797 -0.6118 -0.4012>, rotation=<rot 0 0 90>}]
 dyn_Sleeve_jacket_01_l [{shapeType=Capsule, length=0.5, radius=1, shapeName=dyn_Sleeve_jacket_01_l_capsule, center=<4 0 0>, rotation=<rot 90 0 0>}]
 dyn_Sleeve_jacket_01_r [{shapeType=Capsule, length=2, radius=1, shapeName=dyn_Sleeve_jacket_01_r_capsule, center=<4 0 0>, rotation=<rot 90 0 0>}]
 dyn_Sleeve_jacket_02_l [{shapeType=Capsule, length=1, radius=0.5, shapeName=dyn_Sleeve_jacket_02_l_capsule, center=<4 0 0>, rotation=<rot 90 0 0>}]
 dyn_Sleeve_jacket_02_r [{shapeType=Capsule, length=2, radius=1, shapeName=dyn_Sleeve_jacket_02_r_capsule, center=<4 0 0>, rotation=<rot 90 0 0>}]
 dyn_Sleeve_jacket_03_l [{shapeType=Capsule, length=2.5, radius=0.5, shapeName=dyn_Sleeve_jacket_03_l_capsule, center=<2.1886 -0.9 0>, rotation=<rot 90 0 0>}]
 dyn_Sleeve_jacket_03_r [{shapeType=Capsule, length=2, radius=1, shapeName=dyn_Sleeve_jacket_03_r_capsule, center=<4 0 0>, rotation=<rot 90 0 0>}]
 dyn_Sleeve_jacket_04_l [{shapeType=Capsule, length=1, radius=0.5, shapeName=dyn_Sleeve_jacket_04_l_capsule, center=<0.823 -1.4 0>, rotation=<rot 90 0 0>}]
 dyn_Sleeve_jacket_04_r [{shapeType=Capsule, length=2, radius=1, shapeName=dyn_Sleeve_jacket_04_r_capsule, center=<4 0 0>, rotation=<rot 90 0 0>}]
 dyn_Sleeve_jacket_05_l [{shapeType=Capsule, length=3, radius=0.5, shapeName=dyn_Sleeve_jacket_05_l_capsule, center=<5.4536 0.2563 0>, rotation=<rot 90 0 -10.0002>}]
 dyn_Sleeve_jacket_05_r [{shapeType=Capsule, length=2, radius=1, shapeName=dyn_Sleeve_jacket_05_r_capsule, center=<4 0 0>, rotation=<rot 90 0 0>}]
 dyn_Sleeve_jacket_06_l [{shapeType=Capsule, length=2.6, radius=0.7, shapeName=dyn_Sleeve_jacket_06_l_capsule, center=<5.7195 0 0>, rotation=<rot 90 0 0>}]
 dyn_Sleeve_jacket_06_r [{shapeType=Capsule, length=2, radius=1, shapeName=dyn_Sleeve_jacket_06_r_capsule, center=<4 0 0>, rotation=<rot 90 0 0>}]
 dyn_Sleeve_jacket_07_l [{shapeType=Capsule, length=0.5, radius=1, shapeName=dyn_Sleeve_jacket_07_l_capsule, center=<5.4165 0 0>, rotation=<rot 90 0 0>}]
 dyn_Sleeve_jacket_07_r [{shapeType=Capsule, length=2, radius=1, shapeName=dyn_Sleeve_jacket_07_r_capsule, center=<4 0 0>, rotation=<rot 90 0 0>}]
 dyn_Sleeve_jacket_08_l [{shapeType=Capsule, length=1, radius=0.7, shapeName=dyn_Sleeve_jacket_08_l_capsule, center=<4 0 0>, rotation=<rot 90 0 0>}]
 dyn_Sleeve_jacket_08_r [{shapeType=Capsule, length=2, radius=1, shapeName=dyn_Sleeve_jacket_08_r_capsule, center=<4 0 0>, rotation=<rot 90 0 0>}]
 dyn_back_low_jacket [{shapeType=Capsule, length=1, radius=1, center=<3.4179 0 0>, rotation=<rot 90 0 0>}]
 dyn_side_jacket_01_l [{shapeType=Capsule, length=1, radius=1.5, shapeName=dyn_side_jacket_01_l_capsule, center=<6.3634 0.6388 0.6252>, rotation=<rot 90 0 0>}]
 dyn_side_jacket_01_r [{shapeType=Capsule, length=1, radius=1.5, shapeName=dyn_side_jacket_01_r_capsule, center=<6.42 -0.1775 -0.21>, rotation=<rot 90 0 0>}]
 dyn_upper_jacket_01_l [{shapeType=Capsule, length=1, radius=1}]
 dyn_upper_jacket_01_r [{shapeType=Capsule, length=1, radius=1, center=<0.5 0 0>, rotation=<rot 90 0 0>}]
 dyn_upper_jacket_02_l [{shapeType=Capsule, length=7.7, radius=1.3, center=<3.0219 0.0087 -0.0073>, rotation=<rot -7.7504 9.3608 31.9467>}]
 dyn_upper_jacket_02_r [{shapeType=Capsule, length=7.7, radius=2.3, center=<3.7142 0 0>, rotation=<rot 49.8631 -96.0348 180>}]
 dyn_upper_jacket_03_r [{shapeType=Capsule, length=2, radius=1, shapeName=dyn_upper_jacket_03_r_capsule, center=<4.0399 0 0>, rotation=<rot 0 0 90>}]
 head [{shapeType=Capsule, length=5.2, radius=7.5, shapeName=head_capsule, center=<-0.4956 1.6561 0>, rotation=<rot 0 16.6601 90>}]
 spine_03 [{shapeType=Capsule, length=0.7018, radius=10.2376, shapeName=spine_03_capsule, center=<0 2.8915 0>, rotation=<rot 90 0 0>}]
 spine_05 [{shapeType=Capsule, length=2.3575, radius=9.2129, shapeName=spine_05_capsule, center=<0 1.8844 0>, rotation=<rot 90 0 0>}]
 upperarm_l [{shapeType=Capsule, length=6.6, radius=4.7, shapeName=upperarm_l_capsule, center=<-0.4925 1.2122 5.5095>, rotation=<rot 0 -16.4478 87>}]
 upperarm_twist_01_l [{shapeType=Capsule, length=8, radius=4.1, shapeName=upperarm_twist_01_l_capsule, center=<5.7768 -0.4976 0>, rotation=<rot 90 0 0>}]
 upperarm_twist_01_r [{shapeType=Capsule, length=8, radius=4.1, shapeName=upperarm_twist_01_l_capsule, center=<-5.7768 0 0>, rotation=<rot 90 0 180>}]
}
constraints [
 {
  bone1Name=spine_05
  bone2Name=spine_03
  swing1LimitDegrees=45
  swing1Motion=Locked
  swing2LimitDegrees=45
  swing2Motion=Locked
  twistLimitDegrees=45
  twistMotion=Locked
 }
 {bone1Name=dyn_upper_jacket_01_l, bone2Name=spine_05, swing1LimitDegrees=20, swing2LimitDegrees=5, twistLimitDegrees=2}
 {bone1Name=dyn_upper_jacket_02_l, bone2Name=dyn_upper_jacket_01_l, swing1LimitDegrees=25, swing2LimitDegrees=4, twistLimitDegrees=2}
 {bone1Name=dyn_upper_jacket_01_r, bone2Name=spine_05, swing1LimitDegrees=5, swing2LimitDegrees=5, twistLimitDegrees=2}
 {bone1Name=dyn_upper_jacket_02_r, bone2Name=dyn_upper_jacket_01_r, swing1LimitDegrees=15, swing2LimitDegrees=4, twistLimitDegrees=2}
 {bone1Name=dyn_back_low_jacket, bone2Name=spine_03, swing1LimitDegrees=10, swing2LimitDegrees=20, twistLimitDegrees=2}
 {
  bone1Name=base_Sleeve_jacket_01_l
  bone2Name=upperarm_twist_01_l
  swing1LimitDegrees=45
  swing1Motion=Locked
  swing2LimitDegrees=45
  swing2Motion=Locked
  twistLimitDegrees=45
  twistMotion=Locked
 }
 {bone1Name=dyn_Sleeve_jacket_01_l, bone2Name=base_Sleeve_jacket_01_l, swing1LimitDegrees=10, swing2LimitDegrees=3, twistLimitDegrees=1}
 {
  bone1Name=base_Sleeve_jacket_02_l
  bone2Name=upperarm_twist_01_l
  swing1LimitDegrees=45
  swing1Motion=Locked
  swing2LimitDegrees=45
  swing2Motion=Locked
  twistLimitDegrees=45
  twistMotion=Locked
 }
 {bone1Name=dyn_Sleeve_jacket_02_l, bone2Name=base_Sleeve_jacket_02_l, swing1LimitDegrees=10, swing2LimitDegrees=3, twistLimitDegrees=1}
 {
  bone1Name=base_Sleeve_jacket_03_l
  bone2Name=upperarm_twist_01_l
  swing1LimitDegrees=45
  swing1Motion=Locked
  swing2LimitDegrees=45
  swing2Motion=Locked
  twistLimitDegrees=45
  twistMotion=Locked
 }
 {bone1Name=dyn_Sleeve_jacket_03_l, bone2Name=base_Sleeve_jacket_03_l, swing1LimitDegrees=10, swing2LimitDegrees=3, twistLimitDegrees=1}
 {
  bone1Name=base_Sleeve_jacket_04_l
  bone2Name=upperarm_twist_01_l
  swing1LimitDegrees=45
  swing1Motion=Locked
  swing2LimitDegrees=45
  swing2Motion=Locked
  twistLimitDegrees=45
  twistMotion=Locked
 }
 {bone1Name=dyn_Sleeve_jacket_04_l, bone2Name=base_Sleeve_jacket_04_l, swing1LimitDegrees=10, swing2LimitDegrees=3, twistLimitDegrees=1}
 {
  bone1Name=base_Sleeve_jacket_05_l
  bone2Name=upperarm_twist_01_l
  swing1LimitDegrees=45
  swing1Motion=Locked
  swing2LimitDegrees=45
  swing2Motion=Locked
  twistLimitDegrees=45
  twistMotion=Locked
 }
 {bone1Name=dyn_Sleeve_jacket_05_l, bone2Name=base_Sleeve_jacket_05_l, swing1LimitDegrees=6, swing2LimitDegrees=3, twistLimitDegrees=1}
 {
  bone1Name=base_Sleeve_jacket_06_l
  bone2Name=upperarm_twist_01_l
  swing1LimitDegrees=45
  swing1Motion=Locked
  swing2LimitDegrees=45
  swing2Motion=Locked
  twistLimitDegrees=45
  twistMotion=Locked
 }
 {bone1Name=dyn_Sleeve_jacket_06_l, bone2Name=base_Sleeve_jacket_06_l, swing1LimitDegrees=10, swing2LimitDegrees=3, twistLimitDegrees=1}
 {
  bone1Name=base_Sleeve_jacket_07_l
  bone2Name=upperarm_twist_01_l
  swing1LimitDegrees=45
  swing1Motion=Locked
  swing2LimitDegrees=45
  swing2Motion=Locked
  twistLimitDegrees=45
  twistMotion=Locked
 }
 {bone1Name=dyn_Sleeve_jacket_07_l, bone2Name=base_Sleeve_jacket_07_l, swing1LimitDegrees=10, swing2LimitDegrees=3, twistLimitDegrees=1}
 {
  bone1Name=base_Sleeve_jacket_08_l
  bone2Name=upperarm_twist_01_l
  swing1LimitDegrees=45
  swing1Motion=Locked
  swing2LimitDegrees=45
  swing2Motion=Locked
  twistLimitDegrees=45
  twistMotion=Locked
 }
 {bone1Name=dyn_Sleeve_jacket_08_l, bone2Name=base_Sleeve_jacket_08_l, swing1LimitDegrees=10, swing2LimitDegrees=3, twistLimitDegrees=1}
 {
  bone1Name=upperarm_twist_01_l
  bone2Name=upperarm_l
  swing1LimitDegrees=45
  swing1Motion=Locked
  swing2LimitDegrees=45
  swing2Motion=Locked
  twistLimitDegrees=45
  twistMotion=Locked
 }
 {
  bone1Name=base_Sleeve_jacket_01_r
  bone2Name=spine_05
  swing1LimitDegrees=45
  swing1Motion=Locked
  swing2LimitDegrees=45
  swing2Motion=Locked
  twistLimitDegrees=45
  twistMotion=Locked
 }
 {bone1Name=dyn_Sleeve_jacket_01_r, bone2Name=base_Sleeve_jacket_01_r, swing1LimitDegrees=10, swing2LimitDegrees=3, twistLimitDegrees=1}
 {
  bone1Name=base_Sleeve_jacket_02_r
  bone2Name=spine_05
  swing1LimitDegrees=45
  swing1Motion=Locked
  swing2LimitDegrees=45
  swing2Motion=Locked
  twistLimitDegrees=45
  twistMotion=Locked
 }
 {bone1Name=dyn_Sleeve_jacket_02_r, bone2Name=base_Sleeve_jacket_02_r, swing1LimitDegrees=10, swing2LimitDegrees=3, twistLimitDegrees=1}
 {
  bone1Name=base_Sleeve_jacket_03_r
  bone2Name=spine_05
  swing1LimitDegrees=45
  swing1Motion=Locked
  swing2LimitDegrees=45
  swing2Motion=Locked
  twistLimitDegrees=45
  twistMotion=Locked
 }
 {bone1Name=dyn_Sleeve_jacket_03_r, bone2Name=base_Sleeve_jacket_03_r, swing1LimitDegrees=10, swing2LimitDegrees=3, twistLimitDegrees=1}
 {
  bone1Name=base_Sleeve_jacket_04_r
  bone2Name=spine_05
  swing1LimitDegrees=45
  swing1Motion=Locked
  swing2LimitDegrees=45
  swing2Motion=Locked
  twistLimitDegrees=45
  twistMotion=Locked
 }
 {bone1Name=dyn_Sleeve_jacket_04_r, bone2Name=base_Sleeve_jacket_04_r, swing1LimitDegrees=10, swing2LimitDegrees=3, twistLimitDegrees=1}
 {
  bone1Name=base_Sleeve_jacket_05_r
  bone2Name=spine_05
  swing1LimitDegrees=45
  swing1Motion=Locked
  swing2LimitDegrees=45
  swing2Motion=Locked
  twistLimitDegrees=45
  twistMotion=Locked
 }
 {bone1Name=dyn_Sleeve_jacket_05_r, bone2Name=base_Sleeve_jacket_05_r, swing1LimitDegrees=10, swing2LimitDegrees=3, twistLimitDegrees=1}
 {
  bone1Name=base_Sleeve_jacket_06_r
  bone2Name=spine_05
  swing1LimitDegrees=45
  swing1Motion=Locked
  swing2LimitDegrees=45
  swing2Motion=Locked
  twistLimitDegrees=45
  twistMotion=Locked
 }
 {bone1Name=dyn_Sleeve_jacket_06_r, bone2Name=base_Sleeve_jacket_06_r, swing1LimitDegrees=10, swing2LimitDegrees=3, twistLimitDegrees=1}
 {
  bone1Name=base_Sleeve_jacket_07_r
  bone2Name=spine_05
  swing1LimitDegrees=45
  swing1Motion=Locked
  swing2LimitDegrees=45
  swing2Motion=Locked
  twistLimitDegrees=45
  twistMotion=Locked
 }
 {bone1Name=dyn_Sleeve_jacket_07_r, bone2Name=base_Sleeve_jacket_07_r, swing1LimitDegrees=10, swing2LimitDegrees=3, twistLimitDegrees=1}
 {
  bone1Name=base_Sleeve_jacket_08_r
  bone2Name=spine_05
  swing1LimitDegrees=45
  swing1Motion=Locked
  swing2LimitDegrees=45
  swing2Motion=Locked
  twistLimitDegrees=45
  twistMotion=Locked
 }
 {bone1Name=dyn_Sleeve_jacket_08_r, bone2Name=base_Sleeve_jacket_08_r, swing1LimitDegrees=10, swing2LimitDegrees=3, twistLimitDegrees=1}
 {
  bone1Name=upperarm_twist_01_r
  bone2Name=spine_05
  swing1LimitDegrees=45
  swing1Motion=Locked
  swing2LimitDegrees=45
  swing2Motion=Locked
  twistLimitDegrees=45
  twistMotion=Locked
 }
 {
  bone1Name=upperarm_l
  bone2Name=spine_05
  swing1LimitDegrees=45
  swing1Motion=Locked
  swing2LimitDegrees=45
  swing2Motion=Locked
  twistLimitDegrees=45
  twistMotion=Locked
 }
 {
  bone1Name=base_side_jacket_l
  bone2Name=spine_03
  swing1LimitDegrees=45
  swing1Motion=Locked
  swing2LimitDegrees=45
  swing2Motion=Locked
  twistLimitDegrees=45
  twistMotion=Locked
 }
 {bone1Name=dyn_side_jacket_01_l, bone2Name=base_side_jacket_l, swing1LimitDegrees=10, swing2LimitDegrees=15, twistLimitDegrees=0}
 {
  bone1Name=base_side_jacket_r
  bone2Name=spine_03
  swing1LimitDegrees=45
  swing1Motion=Locked
  swing2LimitDegrees=45
  swing2Motion=Locked
  twistLimitDegrees=45
  twistMotion=Locked
 }
 {bone1Name=dyn_side_jacket_01_r, bone2Name=base_side_jacket_r, swing1LimitDegrees=10, swing2LimitDegrees=15, twistLimitDegrees=0}
 {
  bone1Name=head
  bone2Name=spine_05
  swing1LimitDegrees=45
  swing1Motion=Locked
  swing2LimitDegrees=45
  swing2Motion=Locked
  twistLimitDegrees=45
  twistMotion=Locked
 }
 {
  bone1Name=base_Collar_back_jacket
  bone2Name=spine_05
  swing1LimitDegrees=45
  swing1Motion=Locked
  swing2LimitDegrees=45
  swing2Motion=Locked
  twistLimitDegrees=45
  twistMotion=Locked
 }
 {bone1Name=Collar_back_jacket, bone2Name=base_Collar_back_jacket, swing1LimitDegrees=10, swing2LimitDegrees=1, twistLimitDegrees=5}
 {bone1Name=dyn_upper_jacket_03_r, bone2Name=base_upper_jacket_03_r, swing1LimitDegrees=15, swing2LimitDegrees=15, twistLimitDegrees=1}
 {
  bone1Name=base_upper_jacket_03_r
  bone2Name=spine_05
  swing1LimitDegrees=45
  swing1Motion=Locked
  swing2LimitDegrees=45
  swing2Motion=Locked
  twistLimitDegrees=45
  twistMotion=Locked
 }
 {
  bone1Name=dyn_Pauldron
  bone2Name=upperarm_l
  swing1LimitDegrees=45
  swing1Motion=Free
  swing2LimitDegrees=45
  swing2Motion=Free
  twistLimitDegrees=45
  twistMotion=Free
 }
]
