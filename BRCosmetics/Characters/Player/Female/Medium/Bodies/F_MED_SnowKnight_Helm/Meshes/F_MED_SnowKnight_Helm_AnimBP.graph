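# Post-pose AnimBP for the SnowKnight Helm character body, layering secondary motion onto the parent mesh's copied pose.
# A Control Rig takes progressive-variant helm flags, with Dark and Ultima cross-wired.
# While falling, pauldron and sleeve offsets, two RigidBody simulations and flap offsets apply; otherwise the rig pose passes through.
# Parachute jacket offsets follow, then phase-shifted sine waves bob four floating rocks; the EventGraph updates these drivers and gravity override.
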AnimBlueprint /BRCosmetics/Characters/Player/Female/Medium/Bodies/F_MED_SnowKnight_Helm/Meshes/F_MED_SnowKnight_Helm_AnimBP
target_skeleton=&/F_MED_SnowKnight_Helm_Skeleton
graphs [
 AnimGraph {nodes [
   n1 = Get isFalling() -> (isFalling:double)
   n2 = "Transform (Modify) Bone - Bone: base_flap_01_back_L"!(ComponentPose, Alpha=@n1, Rotation="6,7,6") -> (Pose) {rotationMode=BMM_Additive, alphaBoolBlend {blendOption=Linear}, boneToModify {boneName=base_flap_01_back_L}, rotation=<rot 6 7 6>}
   n3 = "Transform (Modify) Bone - Bone: base_flap_02_back_L"!(ComponentPose=@n2, Alpha=@n1, Rotation="6,10,6") -> (Pose) {rotationMode=BMM_Additive, alphaBoolBlend {blendOption=Linear}, boneToModify {boneName=base_flap_02_back_L}, rotation=<rot 6 10 6>}
   n4 = "Transform (Modify) Bone - Bone: base_flap_01_back_R"!(ComponentPose=@n3, Alpha=@n1, Rotation="8,6,8") -> (Pose) {rotationMode=BMM_Additive, alphaBoolBlend {blendOption=Linear}, boneToModify {boneName=base_flap_01_back_R}, rotation=<rot 8 6 8>}
   n5 = "Transform (Modify) Bone - Bone: base_flap_02_back_R"!(ComponentPose=@n4, Alpha=@n1, Rotation="10,5,10") -> (Pose) {rotationMode=BMM_Additive, alphaBoolBlend {blendOption=Linear}, boneToModify {boneName=base_flap_02_back_R}, rotation=<rot 10 5 10>}
   n6 = "Transform (Modify) Bone - Bone: dyn_Pauldron"!(ComponentPose=@n5, Alpha=Get IsParachuteOpen() -> (IsParachuteOpen:double), Translation="0.000,0.000,-1.507", Rotation="-0.000,-7.604,0.000") -> (Pose) {
    rotationMode=BMM_Additive
    translationMode=BMM_Additive
    alphaBoolBlend {blendOption=Linear}
    alphaScaleBiasClamp {bInterpResult=true}
    boneToModify {boneName=dyn_Pauldron}
    rotation {yaw=-7.604}
    translation {z=-1.507}
   }
   n7 = "Transform (Modify) Bone - Bone: dyn_Pauldron"(
    ComponentPose=@n6
    Alpha=Get ParachuteForward() -> (ParachuteForward:double)
    Translation="0.993,-2.175,1.089"
    Rotation="-0.000,0.000,18.538"
   ) {
    rotationMode=BMM_Additive
    translationMode=BMM_Additive
    alphaBoolBlend {blendOption=Linear}
    alphaScaleBiasClamp {bInterpResult=true}
    boneToModify {boneName=dyn_Pauldron}
    rotation {roll=18.538}
    translation=<0.993 -2.175 1.089>
   }
   n8 = "Transform (Modify) Bone - Bone: parent_Sleeve_jacket_l"!(ComponentPose=@n7, Alpha=Get Sleeve_l() -> (Sleeve_l:double), Rotation="-7.422,7.999,-3.910") -> (Pose) {
    rotationMode=BMM_Additive
    alphaBoolBlend {blendOption=Linear}
    boneToModify {boneName=parent_Sleeve_jacket_l}
    rotation=<rot -7.422 7.999 -3.91>
   }
   "Transform (Modify) Bone - Bone: parent_Sleeve_jacket_r"!(ComponentPose=@n8, Alpha=Get Sleeve_r() -> (Sleeve_r:double), Rotation="-12.288,43.500,5.616") -> (Pose) {
    rotationMode=BMM_Additive
    alphaBoolBlend {blendOption=Linear}
    boneToModify {boneName=parent_Sleeve_jacket_r}
    rotation=<rot -12.288 43.5 5.616>
   }
   n9 = Local To Component(LocalPose=Use cached pose 'EndJacket'())
   n10 = Get Rock_floating_01() -> (Rock_floating_01:Vector)
   n11 = "Transform (Modify) Bone - Bone: dyn_floating_rock_03"(
    ComponentPose="Transform (Modify) Bone - Bone: dyn_floating_rock_02"(
     ComponentPose="Transform (Modify) Bone - Bone: dyn_floating_rock_01"(ComponentPose=@n9, Alpha=1, Translation=@n10, Rotation="0,0,0", Scale="1,1,1") {translationMode=BMM_Additive, alphaBoolBlend {blendOption=Linear}, boneToModify {boneName=dyn_floating_rock_01}}
     Alpha=1
     Translation=Get Rock_floating_02() -> (Rock_floating_02:Vector)
     Rotation="0,0,0"
     Scale="1,1,1"
    ) {translationMode=BMM_Additive, alphaBoolBlend {blendOption=Linear}, boneToModify {boneName=dyn_floating_rock_02}}
    Alpha=1
    Translation=Get Rock_floating_03() -> (Rock_floating_03:Vector)
    Rotation="0,0,0"
    Scale="1,1,1"
   ) {translationMode=BMM_Additive, alphaBoolBlend {blendOption=Linear}, boneToModify {boneName=dyn_floating_rock_03}}
   n12 = Get Rock_floating_04() -> (Rock_floating_04:Vector)
   Output Pose(Result=Component To Local(ComponentPose="Transform (Modify) Bone - Bone: dyn_floating_rock_04"(ComponentPose=@n11, Alpha=1, Translation=@n12, Rotation="0,0,0", Scale="1,1,1") {translationMode=BMM_Additive, alphaBoolBlend {blendOption=Linear}, boneToModify {boneName=dyn_floating_rock_04}})) {layerGroup=None}
   n13 = "Transform (Modify) Bone - Bone: dyn_Pauldron"(
    ComponentPose=Local To Component(LocalPose=Use cached pose 'Start'())
    Alpha=Get IsParachuteOpen()
    Translation="0.000,0.000,-1.507"
    Rotation="-0.000,-7.604,0.000"
   ) {
    rotationMode=BMM_Additive
    translationMode=BMM_Additive
    alphaBoolBlend {blendOption=Linear}
    alphaScaleBiasClamp {bInterpResult=true}
    boneToModify {boneName=dyn_Pauldron}
    rotation {yaw=-7.604}
    translation {z=-1.507}
   }
   n14 = "Transform (Modify) Bone - Bone: dyn_Pauldron"(ComponentPose=@n13, Alpha=Get ParachuteForward(), Translation="0.993,-2.175,1.089", Rotation="-0.000,0.000,18.538") {
    rotationMode=BMM_Additive
    translationMode=BMM_Additive
    alphaBoolBlend {blendOption=Linear}
    alphaScaleBiasClamp {bInterpResult=true}
    boneToModify {boneName=dyn_Pauldron}
    rotation {roll=18.538}
    translation=<0.993 -2.175 1.089>
   }
   n15 = "Transform (Modify) Bone - Bone: parent_Sleeve_jacket_l"(ComponentPose=@n14, Alpha=Get Sleeve_l(), Rotation="-7.422,7.999,-3.910") {
    rotationMode=BMM_Additive
    alphaBoolBlend {blendOption=Linear}
    boneToModify {boneName=parent_Sleeve_jacket_l}
    rotation=<rot -7.422 7.999 -3.91>
   }
   n16 = "Transform (Modify) Bone - Bone: parent_Sleeve_jacket_r"(ComponentPose=@n15, Alpha=Get Sleeve_r(), Rotation="-12.288,43.500,5.616") {
    rotationMode=BMM_Additive
    alphaBoolBlend {blendOption=Linear}
    boneToModify {boneName=parent_Sleeve_jacket_r}
    rotation=<rot -12.288 43.5 5.616>
   }
   n17 = Get Override World Gravity() -> (Override World Gravity:Vector)
   n18 = RigidBody!(ComponentPose=@n16, Alpha, OverrideWorldGravity:Vector=@n17, ExternalForce) -> (Pose) {
    bOverrideWorldGravity=true
    overridePhysicsAsset=&/F_MED_SnowKnight_Helm_Physics
    simulationSpace=BaseBoneSpace
    alphaBoolBlend {blendOption=Linear}
    baseBoneRef {boneName=spine_05}
    componentAppliedLinearAccClamp=<3000 3000 3000>
    componentLinearAccScale=<0.25 0.25 0.25>
    componentLinearVelScale=<0.4 0.4 0.4>
    simSpaceSettings {
     maxAngularAcceleration=1000
     maxAngularVelocity=3
     maxLinearAcceleration=700
     maxLinearVelocity=500
     worldAlpha=0.2
     externalLinearDragV=<0.2 0.2 0.2>
    }
   }
   n19 = RigidBody(ComponentPose=@n18, OverrideWorldGravity=@n17) {
    bOverrideWorldGravity=true
    overridePhysicsAsset=&/F_MED_SnowKnight_Helm_Belt_Physics
    simulationSpace=BaseBoneSpace
    alphaBoolBlend {blendOption=Linear}
    baseBoneRef {boneName=pelvis}
    componentAppliedLinearAccClamp=<3000 3000 3000>
    componentLinearAccScale=<0.25 0.25 0.25>
    componentLinearVelScale=<0.4 0.4 0.4>
    simSpaceSettings {
     maxAngularAcceleration=1000
     maxAngularVelocity=3
     maxLinearAcceleration=700
     maxLinearVelocity=500
     worldAlpha=0.2
     externalLinearDragV=<0.2 0.2 0.2>
    }
   }
   Save cached pose 'EndRB'(Pose=Component To Local(ComponentPose=@n19)) {cacheName=EndRB}
   n20 = Control Rig!(
    Source=Copy Pose From Mesh() {bUseAttachedParent=true}
    Alpha
    Knight Helm loaded:double="To Float (Boolean)"(InBool=Get Knight's Helm loaded() -> (Knight's Helm loaded:bool))
    Royale Helm loaded:double="To Float (Boolean)"(InBool=Get Royale Helm loaded() -> (Royale Helm loaded:bool))
    Spider Helm loaded:double="To Float (Boolean)"(InBool=Get Spider Helm loaded() -> (Spider Helm loaded:bool))
    Ultima Helm:double="To Float (Boolean)"(InBool=Get Dark Helm loaded() -> (Dark Helm loaded:bool))
    Dark Helm loaded:double="To Float (Boolean)"(InBool=Get Ultima Helm() -> (Ultima Helm:bool))
   ) -> (Pose) {
    alphaBoolBlend {blendOption=Linear}
    controlRigAssetReference {blueprintRigClass=&/F_MED_SnowKnight_Helm_DeformRig_ControlRig_new.F_MED_SnowKnight_Helm_DeformRig_ControlRig_new_C}
   }
   Save cached pose 'Start'(Pose=@n20) {cacheName=Start}
   n21 = Get IsParachuteOpen()
   n22 = "Transform (Modify) Bone - Bone: dyn_upper_jacket_01_r"!(
    ComponentPose=Local To Component(LocalPose=Use cached pose 'EndFalling'())
    Alpha=@n21
    Translation="-0.670,0.000,-0.000"
    Rotation="24.939,-5.432,-1.426"
   ) -> (Pose) {
    rotationMode=BMM_Additive
    translationMode=BMM_Additive
    alphaBoolBlend {blendOption=Linear}
    boneToModify {boneName=dyn_upper_jacket_01_r}
    rotation=<rot 24.939 -5.432 -1.426>
    translation {x=-0.67}
   }
   n23 = "Transform (Modify) Bone - Bone: base_upper_jacket_03_r"!(ComponentPose=@n22, Alpha=@n21, Translation="-3.692,0.000,3.353", Rotation="10.000,-5.000,0.000") -> (Pose) {
    rotationMode=BMM_Additive
    translationMode=BMM_Additive
    alphaBoolBlend {blendOption=Linear}
    boneToModify {boneName=base_upper_jacket_03_r}
    rotation {pitch=10, yaw=-5}
    translation {x=-3.692, z=3.353}
   }
   n24 = "Transform (Modify) Bone - Bone: base_upper_jacket_l"!(ComponentPose=@n23, Alpha=@n21, Translation="4,-1,0", Rotation="0,0,0") -> (Pose) {
    rotationMode=BMM_Additive
    translationMode=BMM_Additive
    alphaBoolBlend {blendOption=Linear}
    boneToModify {boneName=base_upper_jacket_l}
    translation=<4 -1>
   }
   Save cached pose 'EndJacket'(Pose=Component To Local(ComponentPose=@n24)) {cacheName=EndJacket}
   n25 = Get isFalling()
   n26 = "Transform (Modify) Bone - Bone: base_flap_01_back_L"(ComponentPose=Local To Component(LocalPose=Use cached pose 'EndRB'()), Alpha=@n25, Rotation="1.354,-6.381,9.051") {
    rotationMode=BMM_Additive
    translationSpace=BCS_BoneSpace
    alphaBoolBlend {blendOption=Linear}
    boneToModify {boneName=base_flap_01_back_L}
    rotation=<rot 1.354 -6.381 9.051>
   }
   n27 = "Transform (Modify) Bone - Bone: base_flap_02_back_L"(ComponentPose=@n26, Alpha=@n25, Rotation="0.231,-2.940,14.958") {
    rotationMode=BMM_Additive
    alphaBoolBlend {blendOption=Linear}
    boneToModify {boneName=base_flap_02_back_L}
    rotation=<rot 0.231 -2.94 14.958>
   }
   n28 = "Transform (Modify) Bone - Bone: base_flap_01_back_R"(ComponentPose=@n27, Alpha=@n25, Rotation="-1.549,1.446,11.371") {
    rotationMode=BMM_Additive
    alphaBoolBlend {blendOption=Linear}
    boneToModify {boneName=base_flap_01_back_R}
    rotation=<rot -1.549 1.446 11.371>
   }
   n29 = "Transform (Modify) Bone - Bone: base_flap_02_back_R"(ComponentPose=@n28, Alpha=@n25, Rotation="-0.463,-2.650,22.375") {
    rotationMode=BMM_Additive
    alphaBoolBlend {blendOption=Linear}
    boneToModify {boneName=base_flap_02_back_R}
    rotation=<rot -0.463 -2.65 22.375>
   }
   Save cached pose 'EndFalling'(Pose=Blend Poses by bool(
     BlendPose_0=Component To Local(ComponentPose=@n29)
     BlendPose_1=Use cached pose 'Start'()
     BlendTime_0=0.2
     BlendTime_1=0.4
     bActiveValue=Get bIsFalling()
    ) {blendType=Linear}) {cacheName=EndFalling}
  ]}
 EventGraph {nodes [
   Event BlueprintUpdateAnimation()
   > n1 = Sequence!(execute) -> (then_0, then_1, then_2:Exec, then_3:Exec, then_4:Exec, then_5:Exec)
   > Set IsParachuteOpen(IsParachuteOpen:double="To Float (Boolean)"(InBool=Get bIsParachuteOpened())) -> (Output_Get:double)
   > Set isSkydiveFloating(isSkydiveFloating:double="To Float (Boolean)"(InBool=Get bIsSkydiveFloating())) -> (Output_Get:double)
   > Set isFalling(isFalling:double="To Float (Boolean)"(InBool=Get bIsFalling())) -> (Output_Get:double)
   n2 = GetRealTimeSeconds()
   n3 = float * float(A=1.4, B=1.3)
   n4 = "Sin (Radians)"(A=float * float(A=@n3, B=@n2))
   Set Rock_floating_01(
    execute=@n1.then_1
    Rock_floating_01:Vector=MakeVector(
     X=Remap Float(Input=@n4, OldRange=(X=-2,Y=0.5), NewRange=(X=-0.5,Y=0.5)) -> (Output=~)
     Y=Remap Float(Input=@n4, OldRange=(X=-1,Y=0.5), NewRange=(X=-0.05,Y=0.5)) -> (Output=~)
     Z=float * float(A=@n4, B=0.2)
    )
   ) -> (Output_Get:Vector)
   n5 = "Sin (Radians)"(A=float * float(A=@n3, B=float - float(A=@n2, B=0.5)))
   > Set Rock_floating_02(Rock_floating_02:Vector=MakeVector(
     X=Remap Float(Input=@n5, OldRange=(X=2,Y=0.3), NewRange=(X=0,Y=-0.5)) -> (Output=~)
     Y=Remap Float(Input=@n5, OldRange=(X=0.1,Y=0.3), NewRange=(X=0.01,Y=-0.02)) -> (Output=~)
     Z=float * float(A=@n5, B=0.25)
    )) -> (Output_Get:Vector)
   n6 = "Sin (Radians)"(A=float * float(A=@n3, B=float - float(A=@n2, B=1)))
   > Set Rock_floating_03(Rock_floating_03:Vector=MakeVector(
     X=Remap Float(Input=@n6, OldRange=(X=-2,Y=0.4), NewRange=(X=-0.1,Y=0.1)) -> (Output=~)
     Y=Remap Float(Input=@n6, OldRange=(X=-8,Y=0.4), NewRange=(X=-1,Y=1)) -> (Output=~)
     Z=float * float(A=@n6, B=0.8)
    )) -> (Output_Get:Vector)
   n7 = "Sin (Radians)"(A=float * float(A=@n3, B=float - float(A=@n2, B=1.5)))
   > Set Rock_floating_04(Rock_floating_04:Vector=MakeVector(X=float * float(A=@n7, B=0.5), Y=float * float(A=@n7, B=0.5), Z=float * float(A=@n7, B=0.5))) -> (Output_Get:Vector)
   n8 = Get Skel Mesh()
   Set Sleeve_l(
    execute=@n1.then_2
    Sleeve_l:double=GetAlphaFromReferenceTransformDelta(SkeletalMeshComponent=@n8, BoneName=clavicle_l, BaseName=upperarm_l, DeltaComponentType=EulerY, InputRange=(X=0,Y=-90), TwistAxis="0,0,1")
   ) -> (Output_Get:double)
   > Set Sleeve_r(Sleeve_r:double=GetAlphaFromReferenceTransformDelta(SkeletalMeshComponent=@n8, BoneName=clavicle_r, BaseName=upperarm_r, DeltaComponentType=EulerY, InputRange=(X=0,Y=-90), TwistAxis="0,0,1")) -> (Output_Get:double)
   n9 = Gravity Override State Function DEPRECATED(
    execute=@n1.then_3
    CharacterPartAnimInstance=Self-Reference()
    GameTimeInSeconds=Get Game Time in Seconds() -> (Game Time in Seconds:double)
    GravityOverrideParams=Get Gravity Override Params()
   )
   > Set Override World Gravity(Override World Gravity:Vector=@n9) -> (Output_Get:Vector)
   n10 = PlayerMovement(execute=@n1.then_4, AnimInstance=Self-Reference(), XDotProductCoefficient=100, YDotProductCoefficient=100, ZDotProductCoefficient=100)
   n11 = "float > float"(A=BreakVector(InVec=MakeVector(X=@n10, Y=@n10.YMovement)).Y, B=40)
   n12 = Get bIsParachuteOpened()
   > Set ParachuteForward(ParachuteForward:double="To Float (Boolean)"(InBool=AND Boolean(A=@n11, B=@n12))) -> (Output_Get:double)
   n13 = Cast To FortPlayerPawn(execute=@n1.then_5, Object=GetOwningActor())
   > n14 = GetVariantChannelActiveIndexForItemActor!(
    execute
    Item:Athena Cosmetic Item Definition Object Reference=/BRCosmetics/Athena/Items/Cosmetics/Characters/Character_SnowKnight_Helm
    Actor:Actor Object Reference=@n13
    Channel:Gameplay Tag Structure=(TagName="Cosmetics.Variant.Channel.Progressive")
    VariantList:Array of Gameplay Tag Structures=Make Array!(
     "[0]"=(TagName="Cosmetics.Variant.Property.Stage1")
     "[1]":Gameplay Tag Structure=(TagName="Cosmetics.Variant.Property.Stage2")
     "[2]":Gameplay Tag Structure=(TagName="Cosmetics.Variant.Property.Stage3")
     "[3]":Gameplay Tag Structure=(TagName="Cosmetics.Variant.Property.Stage4")
     "[4]":Gameplay Tag Structure=(TagName="Cosmetics.Variant.Property.Stage5")
     "[5]":Gameplay Tag Structure=(TagName="")
    ) -> (Array)
    bErrorIfChannelMissing:bool=false
    bErrorIfTagNotInArray:bool=false
   ) -> (then, Index:int=0, bIndexIsValid:bool, bFoundChannel:bool)
   > n15 = Sequence()
   > Set Knight's Helm loaded(Knight's Helm loaded:bool=AND Boolean(A="Equal (Integer)"(A=@n14), B=@n14.bIndexIsValid)) -> (Output_Get:bool)
   Set Royale Helm loaded(execute=@n15.then_1, Royale Helm loaded:bool=AND Boolean(A="Equal (Integer)"(A=@n14, B=1), B=@n14.bIndexIsValid)) -> (Output_Get:bool)
   Set Spider Helm loaded(execute=@n15.then_2, Spider Helm loaded:bool=AND Boolean(A="Equal (Integer)"(A=@n14, B=2), B=@n14.bIndexIsValid)) -> (Output_Get:bool)
   Set Dark Helm loaded(execute=@n15.then_3, Dark Helm loaded:bool=AND Boolean(A="Equal (Integer)"(A=@n14, B=3), B=@n14.bIndexIsValid)) -> (Output_Get:bool)
   Set Ultima Helm(execute=@n15.then_4, Ultima Helm:bool=AND Boolean(A="Equal (Integer)"(A=@n14, B=4), B=@n14.bIndexIsValid)) -> (Output_Get:bool)
   Event BlueprintInitializeAnimation()
   > Set Skel Mesh(Skel Mesh=GetOwnerSkeletalMeshComponent())
  ]}
]
variables [
 Dark Helm loaded {type=boolean}
 Game Time in Seconds {type=number}
 Gravity Override Params {type=GravityOverrideParamsStruct}
 isFalling {type=number}
 IsParachuteOpen {type=number}
 isSkydiveFloating {type=number}
 Knight's Helm loaded {type=boolean}
 Override World Gravity {type=Vector}
 ParachuteForward {type=number}
 Rock_floating_01 {type=Vector, default=<0.6484 0.636 0.1742>}
 Rock_floating_02 {type=Vector, default=<-0.5452 0.003 0.0366>}
 Rock_floating_03 {type=Vector, default=<0.0091 0.7402 -0.5529>}
 Rock_floating_04 {type=Vector, default=<-0.4974 -0.4974 -0.4974>}
 Royale Helm loaded {type=boolean}
 Skel Mesh {type=SkeletalMeshComponent}
 Sleeve_l {type=number}
 Sleeve_r {type=number}
 Spider Helm loaded {type=boolean}
 Ultima Helm {type=boolean}
 varParachutePaudronTranslate {type=Vector}
 varPaudronRotation {type=Vector}
 varSkydivePaudronRotation {type=Vector}
]
